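# Unlit material for the Dinger Inverted vehicle wheel cosmetic, drawing an animated, tinted emissive glow on rim and tire.
# Rim layers combine RGB masks, a panning gradient and a sine pulse, while tire lines Max-blend masked panning distortion, gradient stripes and noise.
# Vertex color red lerps between two layers tinted by Custom Color, then Global Brightness scales the emissive.
# Material attributes pass through shared vehicle damage and de-rez transition functions.
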Material /VehicleCosmetics/Wheels/Dinger_Inverted/Materials/M_Wheel_Dinger_Inverted
graph {
 properties {
  bAutomaticallySetUsageInEditor=false
  bUseMaterialAttributes=true
  bUsedWithMeshDeformer=true
  bUsedWithSkeletalMesh=true
  previewMesh=/VehicleCosmetics/Wheels/Proto09Glitched/Mesh/SK_Wheel_Proto09Glitched
  shadingModel=MSM_Unlit
  usedShadingModels=Unlit
 }
 nodes [
  n1 = NamedRerouteUsage()
  MaterialFunctionCall(In=MaterialFunctionCall(Material=SetMaterialAttributes(Emissive Color=@n1) {attributeSetTypes [#1]}) {materialFunction=&MF_BaseVehicle_Damage}.Result) {materialFunction=&MF_VehicleDeRez_Transition}
  n2 = VectorParameter() {parameterName=Custom Color, defaultValue=<rgba 0.3515 0.5972 1 1>}
  n3 = Multiply(A=@n2.RGB, B=NamedRerouteUsage())
  n4 = Multiply(A=NamedRerouteUsage(), B=@n2.RGB)
  n5 = VertexColor()
  NamedRerouteDeclaration(None=Multiply(A=LinearInterpolate(A=@n3, B=@n4, Alpha=@n5.R), B=ScalarParameter() {defaultValue=1, parameterName=Global Brightness})) {name=Emissive, nodeColor=<rgba 0 1 0.3438 1>}
  n6 = TextureSampleParameter2D() {
   parameterName=Rim Masks
   samplerSource=SSM_Wrap_WorldGroupSettings
   texture=/VehicleCosmetics/Wheels/Dinger_Inverted/Textures/Wheel_Dinger_SE_RGB
  }
  NamedRerouteDeclaration(None=@n6.R) {name=Rim Masks R, nodeColor=<rgba 1 0.3984 0 1>}
  n7 = TextureCoordinate() {uTiling=2}
  n8 = ConstantBiasScale(None=TextureSample(UVs=Panner(Coordinate=@n7) {speedX=-0.5}) {samplerSource=SSM_Wrap_WorldGroupSettings, texture=/VehicleCosmetics/Wheels/_SharedTextures/FX/Gradient_D}.G) {bias=0, scale=2}
  n9 = NamedRerouteUsage()
  n10 = NamedRerouteUsage()
  n11 = Multiply(A=NamedRerouteUsage(), B=Multiply(A=ConstantBiasScale(None=@n10) {bias=0.5, scale=1}) {constB=2}) {constB=2}
  n12 = Multiply(A=Time(), B=ScalarParameter() {defaultValue=0.3, parameterName=RimInfinity Pulse Rate})
  n13 = Multiply(A=ConstantBiasScale(None=Sine(None=@n12)) {bias=1.25, scale=0.4}, B=NamedRerouteUsage())
  NamedRerouteDeclaration(None=Add(A=Multiply(A=@n8, B=@n9), B=Add(A=@n11, B=@n13))) {name=Rim, nodeColor=<rgba 0 0.7891 1 1>}
  NamedRerouteDeclaration(None=@n6.G) {name=Rim Masks G, nodeColor=<rgba 0 1 0 1>}
  NamedRerouteDeclaration(None=@n6.B) {name=Rim Masks B, nodeColor=<rgba 0 0 1 1>}
  n14 = TextureSampleParameter2D() {
   parameterName=Tire Masks
   samplerSource=SSM_Wrap_WorldGroupSettings
   texture=/VehicleCosmetics/Wheels/Dinger_Inverted/Textures/T_DingerInverted_Tire_SE_RGB
  }
  NamedRerouteDeclaration(None=@n14.G) {name=Tire Masks G, nodeColor=<rgba 0 1 0 1>}
  NamedRerouteDeclaration(None=@n14.B) {name=Tire Masks B, nodeColor=<rgba 0 0 1 1>}
  NamedRerouteDeclaration(None=@n14.R) {name=Tire Masks R, nodeColor=<rgba 0.4766 0 1 1>}
  n15 = TextureCoordinate() {uTiling=2, vTiling=0.1}
  n16 = Multiply(
   A=TextureSample(UVs=Panner(Coordinate=@n15) {speedX=0.3, speedY=0.1}) {samplerSource=SSM_Wrap_WorldGroupSettings, texture=/VehicleCosmetics/Wheels/_SharedTextures/FX/Distort_RGB_T}.G
   B=ScalarParameter() {defaultValue=4, parameterName=Tire Emission 2}
  )
  n17 = TextureCoordinate() {uTiling=2}
  n18 = Multiply(
   A=TextureSample(UVs=TextureCoordinate()) {samplerSource=SSM_Wrap_WorldGroupSettings, texture=/VehicleCosmetics/Wheels/_SharedTextures/FX/GradientStripe_T}.G
   B=TextureSample(UVs=Panner(Coordinate=@n17) {speedX=-0.5}) {samplerSource=SSM_Wrap_WorldGroupSettings, texture=&Gradient_D}.G
  )
  n19 = Multiply(A=Max(A=Multiply(A=@n16) {constB=0.4}, B=Multiply(A=@n18) {constB=2}), B=NamedRerouteUsage())
  n20 = TextureCoordinate() {vTiling=0.25}
  n21 = Multiply(
   A=TextureSample(UVs=Panner(Coordinate=@n20) {speedX=-0.1, speedY=-0.05}) {samplerSource=SSM_Wrap_WorldGroupSettings, texture=&Distort_RGB_T}.B
   B=ScalarParameter() {defaultValue=4, parameterName=Tire Emission 1}
  )
  n22 = Multiply(A=Multiply(A=Clamp(None=@n21) {minDefault=0.1}, B=NamedRerouteUsage())) {constB=1.2}
  n23 = Clamp(None=@n21) {minDefault=0.2}
  n24 = TextureCoordinate() {uTiling=32, vTiling=3}
  n25 = Multiply(A=Multiply(
    A=TextureSample(UVs=@n24) {samplerSource=SSM_Wrap_WorldGroupSettings, texture=/VehicleCosmetics/Wheels/_SharedTextures/FX/Universal_Noise_14_RGB}.G
    B=NamedRerouteUsage()
   )) {constB=1.2}
  NamedRerouteDeclaration(None=Max(A=Max(A=@n19, B=@n22), B=Multiply(A=@n23, B=@n25))) {name=Tire Lines, nodeColor=<rgba 1 0 0.7031 1>}
  NamedRerouteDeclaration(None=@n21) {name=Distort, nodeColor=<rgba 0 1 0.4141 1>}
 ]
}
